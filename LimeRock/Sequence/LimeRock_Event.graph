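# LimeRock_Event is a Level Sequence for LimeRock that acts as a thin wrapper around nested sequence content.
# It has no object bindings of its own, only one subsequence track named LimeRock_100.
# That track holds a single short section, running from 0.15 to 0.6 seconds on a 60 fps timeline.
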LevelSequence /LimeRock/Sequence/LimeRock_Event
bindings []
display_rate {denominator=1, numerator=60}
tick_resolution {denominator=1, numerator=24000}
tracks [{type=MovieSceneSubTrack, display_name=LimeRock_100, sections [{type=MovieSceneSubSection, end_frame=14400, start_frame=3600}]}]
