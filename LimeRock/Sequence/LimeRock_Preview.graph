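# LimeRock_Preview is a Level Sequence that wraps the LimeRock_Event sequence, apparently to preview it alongside an animated sphere.
# A subsequence track plays LimeRock_Event over a range that spans all of the sphere's keys.
# The possessed FortStaticMeshActor sphere, heavily scaled up, follows a keyed path of location and rotation through the world.
# A visibility track shows the sphere shortly after its motion begins and hides it after the motion ends.
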LevelSequence /LimeRock/Sequence/LimeRock_Preview
bindings [Sphere {
  id=#1
  possessed_class=FortStaticMeshActor
  tracks [
   {
    type=MovieScene3DTransformTrack
    sections [{
      type=MovieScene3DTransformSection
      channels [
       Location.X {
        type=MovieSceneScriptingDoubleChannel
        default=-126385.2569
        thinned_from=16
        keys [[7568,-143315.7144],[7617,-138031.6675],[7652,-134968.4335],[7843,-126298.5219],[8040,-121323.2236],[8287,-108289.4899],[8391,-107124.3427],[8480,-111781.0936],[8520,-113517.552],[8595,-114515.7928],[8657,-114515.7915],[8839,-108799.9182],[9055,-108799.9182],[9164,-106244.0573],[9284,-107095.8897]]
       }
       Location.Y {
        type=MovieSceneScriptingDoubleChannel
        default=-95370.1841
        thinned_from=19
        keys [[7568,-108090.8001],[7652,-108090.8001],[7860,-107019.7863],[8113,-104199.1751],[8201,-102562.2946],[8340,-98621.6994],[8383,-97237.1498],[8430,-96634.0324],[8518,-97071.6622],[8602,-96715.2884],[8657,-96763.8828],[8863,-94092.5499],[8963,-93693.9648],[9043,-93588.7262],[9167,-92137.6565],[9196,-92176.7647]]
       }
       Location.Z {
        type=MovieSceneScriptingDoubleChannel
        default=18966
        thinned_from=25
        keys [[7568,2987.7542],[7669,3728.7775],[7758,6332.9637],[7851,11382.1212],[8023,14733.6988],[8163,13852.9485],[8240,9697.0172],[8253,9575.0342],[8292,10100.6447],[8338,10360.7627],[8401,11795.7601],[8508,18019.7056],[8542,18952.8286],[8571,19113.875],[8663,18703.4451],[8756,16559.1996],[8780,16350.6107],[8843,16753.0847],[8909,17083.008],[8960,16933.3215],[9096,15910.4361],[9199,16847.5552],[9353,16170.8219]]
       }
       Rotation.X {
        type=MovieSceneScriptingDoubleChannel
        default=0
        thinned_from=17
        keys [[7568,0],[8155,0],[8218,-4.5979],[8271,-1.4567],[8330,-3.2708],[8355,-4.2145],[8452,-0.68],[8524,0],[8636,0],[8775,7.1329],[8926,13.0799],[9058,5.8419],[9186,23.5859],[9295,15.7686]]
       }
       Rotation.Y {
        type=MovieSceneScriptingDoubleChannel
        default=0
        thinned_from=18
        keys [[7568,0],[7963,0],[8091,-5.0729],[8202,-16.451],[8314,-20.7977],[8377,-18.6885],[8476,-9.6118],[8517,-4.6253],[8587,0.1492],[8656,-2.0888],[8782,-25.5606],[8793,-26.5587],[8947,-11.5553],[9021,-9.6417],[9179,-47.4731]]
       }
       Rotation.Z {
        type=MovieSceneScriptingDoubleChannel
        default=0
        thinned_from=10
        keys [[7568,15.8039],[8260,15.8039],[8500,4.5636],[8617,2.1088],[8906,-0.2803],[9043,-3.3854],[9227,-18.8038]]
       }
       Scale.X {type=MovieSceneScriptingDoubleChannel, default=60}
       Scale.Y {type=MovieSceneScriptingDoubleChannel, default=60}
       Scale.Z {type=MovieSceneScriptingDoubleChannel, default=74.6732}
      ]
     }]
   }
   {
    type=MovieSceneVisibilityTrack
    sections [{
      type=MovieSceneVisibilitySection
      channels [None {type=MovieSceneScriptingBoolChannel, default=false, keys [[7752,false],[7758,true],[10198,false]]}]
     }]
   }
  ]
 }]
display_rate {denominator=1, numerator=60}
tick_resolution {denominator=1, numerator=24000}
tracks [{type=MovieSceneSubTrack, display_name=LimeRock_Event, sections [{type=MovieSceneSubSection, end_frame=14400, start_frame=3600}]}]
